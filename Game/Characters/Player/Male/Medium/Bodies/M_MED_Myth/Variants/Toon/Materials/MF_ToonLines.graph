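# Material function in the Myth Toon character variant's folder that draws toon lines by darkening Base Color and Emissive of incoming material attributes.
# A remapped Fresnel on the transformed material normal forms the outline, multiplied by ink lines from a channel-selected, inverted, powered edge texture.
# Both layers have opacity parameters.
# Forward and mobile paths always use the Fresnel outline, while the default path includes it only through the UseFresnelOutlineonHighend static switch.
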MaterialFunction /Game/Characters/Player/Male/Medium/Bodies/M_MED_Myth/Variants/Toon/Materials/MF_ToonLines
graph {nodes [
  Constant() {r=1}
  n1 = GetMaterialAttributes(None=FunctionInput() {inputName=Material, inputType=FunctionInput_MaterialAttributes}) {attributeGetTypes [#1, #2, #3]}
  n2 = ScalarParameter() {defaultValue=-4, group=ToonLines, parameterName=Toon_OutlineC1}
  n3 = ScalarParameter() {defaultValue=12, group=ToonLines, parameterName=Toon_OutlineC2}
  n4 = ScalarParameter() {defaultValue=6, group=ToonLines, parameterName=Toon_OutlineExponent1}
  n5 = Transform(None=@n1.Normal)
  n6 = Power(
   Base=OneMinus(None=Fresnel(ExponentIn=@n4, Normal=@n5) {baseReflectFraction=0})
   Exp=ScalarParameter() {defaultValue=12, group=ToonLines, parameterName=Toon_OutlineExponent2}
  )
  n7 = LinearInterpolate(
   B=Saturate(None=LinearInterpolate(A=@n2, B=@n3, Alpha=@n6))
   Alpha=ScalarParameter() {defaultValue=1, group=ToonLines, parameterName=Toon_Line_Opacity}
  ) {constA=1}
  n8 = StaticSwitchParameter(True=@n7, False=Constant() {r=1}) {parameterName=UseFresnelOutlineonHighend}
  n9 = ChannelMaskParameter(None=TextureSampleParameter2D() {group=ToonLines, parameterName=Toon_EdgeLineTexture, texture=&T_WhitesRGB}.RGB) {group=ToonLines, maskChannel=Alpha, parameterName=Toon_EdgeLineChannel, defaultValue=<rgba 0 0 0 1>}
  n10 = OneMinus(None=@n9.R)
  n11 = ScalarParameter() {defaultValue=4, group=ToonLines, parameterName=Toon_LineThicknessPower}
  n12 = Multiply(
   A=ShadingPathSwitch(Default=QualitySwitch(Default=@n8), Forward=@n7, Mobile=@n7)
   B=LinearInterpolate(
    B=Power(Base=@n10, Exp=@n11) {constExponent=3}
    Alpha=ScalarParameter() {defaultValue=1, group=ToonLines, parameterName=Toon_InkLine_Opacity}
   ) {constA=1}
  )
  FunctionOutput(None=SetMaterialAttributes!(MaterialAttributes=@n1.MaterialAttributes, BaseColor=Multiply(A=@n1.BaseColor, B=@n12), EmissiveColor=Multiply(A=@n1.EmissiveColor, B=@n12)) {attributeSetTypes [#3, #2]})
  VertexNormalWS()
 ]}
